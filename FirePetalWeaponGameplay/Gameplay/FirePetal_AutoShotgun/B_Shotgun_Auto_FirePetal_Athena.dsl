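; Fire Petal auto shotgun weapon Blueprint for Athena, extending its parent's impact and firing effects.
; On impact, it passes its DebrisAmount variable to the spawned particle system.
; On fire, it waits briefly (a delay marked as required), deactivates all shell effects, then activates shell ejection at the secondary socket.
; That ejection reuses the inherited shell settings, with reversed rotation rate and slightly reduced gravity.
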(event FortWeapon|EventOnPlayImpactFX
  (params (HitResult :type Hit%20Result%20Structure%20%28by%20ref%29) (ImpactPhysicalSurface :type EPhysicalSurface%20Enum :default "SurfaceType_Default") (SpawnedPSC :type FXSystem%20Component%20Object%20Reference))
  (|CallParentFunction HitResult ImpactPhysicalSurface SpawnedPSC)
  (Effects|Components|ParticleSystem|SetFloatParameter SpawnedPSC "DebrisAmount" (Variables|Default|GetDebrisAmount)))

(event EventOnPlayWeaponFireFX
  (params (bPersistentFire :type Boolean) (bSecondaryFire :type Boolean))
  (|CallParentFunction bPersistentFire bSecondaryFire)
  (Utilities|FlowControl|Delay 0.2) ; REQUIRED to make work!
  (CallFunction|DeactivateAllShellsFX)
  (CallFunction|ActivateShellsFXNative (Variables|Effects|Shells|GetShell%20Burst%20FX) (Variables|Effects|Shells|GetShell%20Looping%20FX) (Variables|Default|GetSecondary%20Shells%20Socket%20Name) "ShellsRotationRate" (* (Variables|Effects|Shells|GetShellsRotationRate) "-1.000000,-1.000000,-1.000000") "Velocity" (* (Variables|Effects|Shells|GetShells%20Velocity) "1.000000,1.000000,1.000000") "Gravity" (* (Variables|Effects|Shells|GetShells%20Gravity) "0.970000,0.970000,0.970000") "Size" (Variables|Effects|Shells|GetShells%20Size) "InheritParentVelocity" (Math|Conversions|ToVector(Float) (Variables|Effects|Shells|GetInherit%20Parent%20Velocity)) "CylinderRadius" (Variables|Effects|Shells|GetCylindrical%20Radius) "CylinderHeight" (Variables|Effects|Shells|GetCylindrical%20Height) "ShellColor" (Variables|Effects|Shells|GetShell%20Color) "Shell Type" (Utilities|Select :Index (Variables|Effects|Shells|GetShellTypeSelect) :NewEnumerator0 0 :NewEnumerator1 1 :NewEnumerator2 2 :NewEnumerator3 4 :NewEnumerator4 3) (Variables|Effects|DrawDistances|GetShells%20PS%20Max%20Draw%20Distance) (Variables|Effects|Shells|GetUseShellsOnReload?) "SpawnProbability" (Variables|Effects|Shells|GetShells%20Spawn%20Rate%20Scale) "IsSystemLocal" (Variables|Default|GetbIsLocal) "UseShellSmoke" (Variables|Effects|Shells|GetbUseShellSmoke)))

(fn ConstructionScript
  (|CallParentFunction))
