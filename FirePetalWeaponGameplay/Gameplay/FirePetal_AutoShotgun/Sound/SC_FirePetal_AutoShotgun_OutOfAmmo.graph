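# Sound cue for the FirePetal auto shotgun's out-of-ammo sound.
# A random node picks between two equally weighted no-ammo waves and feeds a root modulator left at default settings.
# It uses the weapon foley sound class and the shared weapon reload attenuation, overrides concurrency, and has its debug flag left on.
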SoundCue /FirePetalWeaponGameplay/Gameplay/FirePetal_AutoShotgun/Sound/SC_FirePetal_AutoShotgun_OutOfAmmo
root=Modulator_0
nodes [
 {id=Modulator_0, children [Random_0]}
 {id=Random_0, children [WavePlayer_0,WavePlayer_3], properties {weights [1,1]}}
 {
  id=WavePlayer_0
  properties {soundWaveAssetPtr=/FirePetalWeaponGameplay/Gameplay/FirePetal_AutoShotgun/Sound/SoundWaves/FirePetal_AutoShotgun_No_Ammo_01}
 }
 {
  id=WavePlayer_3
  properties {soundWaveAssetPtr=/FirePetalWeaponGameplay/Gameplay/FirePetal_AutoShotgun/Sound/SoundWaves/FirePetal_AutoShotgun_No_Ammo_02}
 }
]
properties {
 attenuationSettings=&Fort_Weapon_Reload_Att
 bDebug=true
 bOverrideConcurrency=true
 maxDistance=2250
 soundClassObject=&Fort_WeaponFoley
 volumeMultiplier=1
}
